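; Passive quest ability that tracks the player holding a weapon at max charge for the BattlePassS21 "Hold Charge For Five" quest.
; On activation it listens for the max charge tag on the player controller being added and removed.
; After addition, a 2.9-second delay precedes a pawn tag check that applies the quest GameplayEffect or ends the ability.
; On removal it applies the Remove GE and ends; the weapon-fired wait leads nowhere.
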(event Ability|EventOnEndAbility
  (params (bWasCancelled :type Boolean))
  (Ability|EndAbility))

(event Ability|EventFailedToActivatePassiveAbility
  (params (ActorInfo :type Gameplay%20Ability%20Actor%20Info%20Structure))
  (Ability|EndAbility))

(event Ability|EventActivateAbility
  (|CallParentFunction)
  (Variables|Default|SetPlayer%20Controller (Fortnite|GetFortPlayerControllerfromActor (Variables|GAVars|Data|GetOwningPlayerPawn)))
  (Ability|Tasks|WaitGameplayTagAddToActor (Variables|Default|GetPlayer%20Controller) (Variables|Default|GetMax%20Charge%20Tag)
    (:then
      (Ability|Tasks|WaitGameplayTagRemoveFromActor (Variables|Default|GetPlayer%20Controller) (Variables|Default|GetMax%20Charge%20Tag)
        (:Removed
          (if (GameplayTags|HasMatchingGameplayTag (Variables|GAVars|Data|GetOwningPlayerPawn) (Variables|Default|GetWeapon%20Fired%20Tag))
            (Ability|Tasks|WaitGameplayTagRemoveFromActor (Variables|Default|GetPlayer%20Controller) (Variables|Default|GetWeapon%20Fired%20Tag)
              (:Removed
                _))
            (else
              _))
          (Creative|GameplayEffects|SimpleApplyGameplayEffectToActor (Variables|Default|GetRemove%20GE) (Variables|Default|GetPlayer%20Controller))
          (Ability|EndAbility))))
    (:Added
      (Utilities|FlowControl|RetriggerableDelay 2.9)
      (if (GameplayTags|HasMatchingGameplayTag (Variables|GAVars|Data|GetOwningPlayerPawn) (Variables|Default|GetMax%20Charge%20Tag))
        (Creative|GameplayEffects|SimpleApplyGameplayEffectToActor "/BattlePassS21/Items/QuestAbilities/Quest_HoldChargeForFive/GE_Quest_HoldChargeForFive.GE_Quest_HoldChargeForFive_C" (Variables|Default|GetPlayer%20Controller))
        (else
          (Ability|EndAbility))))))
