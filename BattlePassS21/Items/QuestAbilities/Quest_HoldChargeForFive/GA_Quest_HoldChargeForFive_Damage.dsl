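; Passive quest ability for the "Hold Charge For Five" Battle Pass quest that credits progress when the player deals qualifying damage.
; On activation it listens repeatedly for damage-dealt gameplay events.
; Each event requires the owning pawn to carry the quest tag and the target to match any tag in a configurable array.
; Passing hits send a custom stat event to the player's quest manager.
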(event Ability|EventOnEndAbility
  (params (bWasCancelled :type Boolean))
  (Ability|EndAbility))

(event Ability|EventFailedToActivatePassiveAbility
  (params (ActorInfo :type Gameplay%20Ability%20Actor%20Info%20Structure))
  (Ability|EndAbility))

(event Ability|EventActivateAbility
  (|CallParentFunction)
  (bind (_asynctaskproxy _payload) (Ability|Tasks|WaitGameplayEvent "(TagName=\"Event.Damage.Dealt\")" 0 false false)
    (:EventReceived
      (if (GameplayTags|HasMatchingGameplayTag (Variables|GAVars|Data|GetOwningPlayerPawn) "(TagName=\"Athena.Quests.HoldChargeForFive\")")
        (bind _returnvalue (GameplayTags|HasAnyMatchingGameplayTags(Message) (.Target (Utilities|Struct|BreakGameplayEventData _payload)) (GameplayTags|MakeGameplayTagContainerfromArray (Variables|Default|GetGameplay%20Tags))))
        (if _returnvalue
          (Quest|SendComplexCustomStatEvent (Variables|GAVars|Data|GetOwningPlayerQuestManager) 0 "(GameplayTags=((TagName=\"Athena.Quests.HoldChargeForFive\")))" "(GameplayTags=)" 1))))))
